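# Sound cue for the Sahara Male cosmetic pickaxe, played, per its name, when it hits an enemy.
# After a fixed short delay, a mixer layers one of six random impact waves with an optional electric accent that has its own attenuation node.
# The accent randomizer pads four electric waves with four heavier-weighted empty slots, so it sounds about one hit in three.
# The cue uses shared pickaxe-impact attenuation, property sheet and impacts sound class.
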SoundCue /BRCosmetics/Sounds/PickAxes/Sahara_Male/PickaxeImpactEnemy_Sahara_Male_Cue1
root=Delay_0
nodes [
 {id=Attenuation_0, children [Random_1]}
 {id=Delay_0, children [Mixer_0], properties {delayMax=0.1, delayMin=0.1}}
 {id=Mixer_0, children [Attenuation_0, Random_0], properties {inputVolume [1,1]}}
 {id=Random_0, children [WavePlayer_0,WavePlayer_1,WavePlayer_2,WavePlayer_3,WavePlayer_4,WavePlayer_5], properties {weights [1,1,1,1,1,1]}}
 {
  id=Random_1
  children [WavePlayer_6,WavePlayer_7,WavePlayer_8,WavePlayer_9,null,null,null,null]
  properties {weights [0.5,0.5,0.5,0.5,1,1,1,1]}
 }
 {id=WavePlayer_0, properties {soundWaveAssetPtr=&/Pickaxe_SaharaMale_Impact_06}}
 {id=WavePlayer_1, properties {soundWaveAssetPtr=&/Pickaxe_SaharaMale_Impact_05}}
 {id=WavePlayer_2, properties {soundWaveAssetPtr=&/Pickaxe_SaharaMale_Impact_04}}
 {id=WavePlayer_3, properties {soundWaveAssetPtr=&/Pickaxe_SaharaMale_Impact_03}}
 {id=WavePlayer_4, properties {soundWaveAssetPtr=&/Pickaxe_SaharaMale_Impact_02}}
 {id=WavePlayer_5, properties {soundWaveAssetPtr=&/Pickaxe_SaharaMale_Impact_01}}
 {id=WavePlayer_6, properties {soundWaveAssetPtr=&/Pickaxe_SaharaMale_Electric_Impact_04}}
 {id=WavePlayer_7, properties {soundWaveAssetPtr=&/Pickaxe_SaharaMale_Electric_Impact_01}}
 {id=WavePlayer_8, properties {soundWaveAssetPtr=&/Pickaxe_SaharaMale_Electric_Impact_02}}
 {id=WavePlayer_9, properties {soundWaveAssetPtr=&/Pickaxe_SaharaMale_Electric_Impact_03}}
]
properties {
 attenuationSettings=/Game/Sounds/Attenuation/Fort_Pickaxe_Impact_Att
 audioPropertiesSheet=/BRCosmetics/Sounds/Settings/PropertySheets/Pickaxe/APS_Cosmetic_Pickaxe_Impact
 maxDistance=2097152
 soundClassObject=&Fort_Impacts
 volumeMultiplier=0.615
 concurrencySet [null]
}
